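# Translucent UI material for the LemonExam reticle, drawing a bracket shape and a center pip from scalar and color parameters.
# Mirrored UVs sample MSDF textures for brackets with a patterned fill, and a Step on Progress creates the fill mask.
# Pip_Ranged_ON blends between a rotated-box pip with shadow and a ranged dot with background.
# Named reroutes carry anti-aliased SDF masks into color and opacity composites, with MasterOpacity scaling the opacity.
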Material /LemonExam/UI/Materials/M_UI_LemonExam_Reticle
graph {
 properties {bAutomaticallySetUsageInEditor=false, blendMode=BLEND_Translucent, materialDomain=MD_UI, usedShadingModels=Default Lit}
 nodes [
  Constant() {r=0.1}
  Constant() {r=0.1}
  Constant() {r=1.2}
  n1 = VectorParameter() {parameterName=BracketDefaultColor, defaultValue=<rgba 0.1867 0.483 0.7865 1>}
  n2 = NamedRerouteUsage()
  n3 = Multiply(A=NamedRerouteUsage(), B=ScalarParameter() {defaultValue=0.5, parameterName=BracketsFill_Opacity}) {constB=2}
  n4 = NamedRerouteUsage()
  LinearInterpolate(A=LinearInterpolate(A=LinearInterpolate(B=@n1.RGB, Alpha=@n2), Alpha=LinearInterpolate(B=@n3, Alpha=@n4)), Alpha=NamedRerouteUsage())
  MaterialFunctionCall() {materialFunction=&DebugTimeSine}
  Max(B=LinearInterpolate(A=NamedRerouteUsage()) {constB=-2})
  n5 = NamedRerouteUsage()
  Max(A=@n5, B=LinearInterpolate(A=@n5) {constB=-2})
  Multiply(B=ScalarParameter() {defaultValue=2, parameterName=PipOpacity}) {constB=2}
  n6 = LinearInterpolate(
   A=ScalarParameter() {defaultValue=0.7, parameterName=BracketDefault_Opacity}
   B=ScalarParameter() {defaultValue=1, parameterName=Fill_Opacity}
   Alpha=@n4
  )
  n7 = NamedRerouteUsage()
  n8 = NamedRerouteUsage()
  n9 = NamedRerouteUsage()
  n10 = ScalarParameter() {defaultValue=1, parameterName=Pip_Ranged_ON}
  Multiply(
   A=LinearInterpolate(A=Multiply(A=@n6, B=@n7), Alpha=LinearInterpolate(A=@n8, B=@n9, Alpha=@n10)) {constA=1}
   B=ScalarParameter() {defaultValue=1, parameterName=MasterOpacity}
  )
  n11 = MaterialFunctionCall() {materialFunction=&GetUserInterfaceUV}
  n12 = ScalarParameter() {defaultValue=0.055, parameterName=PipScale_Shadow}
  n13 = LinearInterpolate(A=@n12, B=@n12) {constA=0.125, constB=0.125}
  n14 = MaterialFunctionCall(
   UVs=MaterialFunctionCall(
    UVs=MaterialFunctionCall(UVs=@n11.Normalized UV, Texture Scale=@n13) {materialFunction=&ScaleUVsByCenter}.UVs
    "Rotation Angle (0-1)"=Constant() {r=0.125}
   ) {materialFunction=&CustomRotator}.Rotated Values
   Size=Constant() {r=0.58}
   OutlineThickness=Constant() {r=0.1}
  ) {materialFunction=&MF_UI_Box}
  NamedRerouteDeclaration(None=@n14.Outline) {name=PipShadow, nodeColor=<rgba 0 1 0.3672 1>}
  n15 = ScalarParameter() {defaultValue=0.2, parameterName=PipScale_BGRanged}
  n16 = LinearInterpolate(A=@n15, B=@n15) {constA=0.125, constB=0.125}
  n17 = MaterialFunctionCall!(MSDF RGB=TextureSample(UVs=MaterialFunctionCall(UVs=@n11.Normalized UV, Texture Scale=@n16) {materialFunction=&ScaleUVsByCenter}.UVs) {samplerType=SAMPLERTYPE_LinearColor, texture=/LemonExam/UI/Textures/T_UI_LemonExam_dot}.RGB) {materialFunction=/FortUILibrary/MaterialFunctions/Shapes/Modifiers/MF_UI_MSDFtoSDF}
  NamedRerouteDeclaration(None=MaterialFunctionCall(DistanceField=@n17.SDF, Threshold=Constant() {r=0.58}, Smoothing=Constant() {r=0.1}) {materialFunction=&MF_UI_AntiAliasing}.Outer) {name=PipRangedBG, nodeColor=<rgba 1 0.0469 0 1>}
  n18 = VectorParameter() {parameterName=BracketFillColor, defaultValue=<rgba 0.6333 0.1899 0.9829 1>}
  n19 = MaterialFunctionCall() {materialFunction=&GetUserInterfaceUV}
  n20 = AppendVector(A=Constant() {r=0.8}, B=Constant() {r=0.2})
  n21 = Constant()
  n22 = Constant() {r=-0.2}
  n23 = MaterialFunctionCall!(
   UV=Add(
    A=MaterialFunctionCall(UVs=@n19.Normalized UV, Texture Scale=@n20) {materialFunction=&ScaleUVsByCenter}.UVs
    B=AppendVector(A=@n21, B=@n22)
   ) {constB=-0.2}
   Mirror Y
   Swizzle UV
  ) {materialFunction=/FortUILibrary/MaterialFunctions/Basics/MF_UI_Symmetry_2x1_New}
  n24 = MaterialFunctionCall(MSDF RGB=TextureSample(UVs=@n23.Flipped Symmetry) {samplerType=SAMPLERTYPE_LinearColor, texture=/LemonExam/UI/Textures/T_UI_LemonExam_Reticle_MSDF}.RGB) {materialFunction=&MF_UI_MSDFtoSDF}
  n25 = NamedRerouteDeclaration(None=MaterialFunctionCall(DistanceField=@n24.SDF, Threshold=Constant() {r=0.45}, Smoothing=Constant() {r=0.02}) {materialFunction=&MF_UI_AntiAliasing}.Outer) {name=Bracket, nodeColor=<rgba 0 0.2266 1 1>}
  n26 = MaterialFunctionCall(UVs=Add(A=@n23.Flipped Symmetry) {constB=0.15}, Texture Scale=Constant2Vector() {g=1.4, r=1.4}) {materialFunction=&ScaleUVsByCenter}
  n27 = MaterialFunctionCall(MSDF RGB=TextureSample(UVs=MaterialFunctionCall!(UVs=@n26.UVs) {materialFunction=/FortUILibrary/MaterialFunctions/Basics/MF_UI_Flip}.Mirror Y) {samplerType=SAMPLERTYPE_LinearColor, texture=/LemonExam/UI/Textures/T_UI_LemonExam_pattern2}.RGBA) {materialFunction=&MF_UI_MSDFtoSDF}
  n28 = ScalarParameter() {defaultValue=0.53, parameterName=BracketPattern_Threshold}
  n29 = ScalarParameter() {defaultValue=0.3, parameterName=BracketPatternBlur}
  n30 = Multiply(A=NamedRerouteDeclaration(None=MaterialFunctionCall(DistanceField=@n27.SDF, Threshold=@n28, Smoothing=@n29) {materialFunction=&MF_UI_AntiAliasing}.Outer) {name=BracketPattern, nodeColor=<rgba 0.5469 1 0 1>})
  n31 = ComponentMask(None=@n25) {r=true}
  n32 = Multiply(
   A=LinearInterpolate(
    B=VectorParameter() {parameterName=BracketPatternColor, defaultValue=<rgba 0.7135 0.9664 1 1>}.RGB
    Alpha=LinearInterpolate(B=@n30, Alpha=@n31)
   )
   B=ScalarParameter() {defaultValue=1, parameterName=BracketPatternOpacity}
  )
  NamedRerouteDeclaration(None=LinearInterpolate(
    A=LinearInterpolate(B=@n18.RGB, Alpha=@n25)
    Alpha=NamedRerouteDeclaration(None=@n32) {name=Bracket_PatternFill, nodeColor=<rgba 1 0 0.9609 1>}
   )) {name=Brackets_Fill_Final, nodeColor=<rgba 1 0.8906 0 1>}
  n33 = ScalarParameter() {defaultValue=0.05, parameterName=PipScale}
  n34 = MaterialFunctionCall(UVs=@n11.Normalized UV, Texture Scale=LinearInterpolate(A=@n33, B=@n33) {constA=0.125, constB=0.125}) {materialFunction=&ScaleUVsByCenter}
  n35 = Constant() {r=0.125}
  NamedRerouteDeclaration(None=MaterialFunctionCall(UVs=MaterialFunctionCall(UVs=@n34.UVs, "Rotation Angle (0-1)"=@n35) {materialFunction=&CustomRotator}.Rotated Values) {materialFunction=&MF_UI_Box}.Fill) {name=PipCenter, nodeColor=<rgba 0.6425 0.244 1 1>}
  n36 = ScalarParameter() {defaultValue=0.5, parameterName=Progress}
  NamedRerouteDeclaration(None=Step(Y=ComponentMask(None=@n23.Flipped Symmetry) {r=true}, X=LinearInterpolate(Alpha=@n36) {constA=0.03, constB=0.87}) {constX=0}) {name=ProgressFIll, nodeColor=<rgba 0 0.4844 1 1>}
  n37 = VectorParameter() {parameterName=PipDefaultShadow, defaultValue=<rgba 0.0091 0.1401 0.3132 1>}
  n38 = NamedRerouteUsage()
  n39 = LinearInterpolate(
   A=LinearInterpolate(B=VectorParameter() {parameterName=PipCenterColor, defaultValue=<rgba 1 1 1 1>}.RGB, Alpha=NamedRerouteUsage())
   Alpha=Multiply(A=LinearInterpolate(B=@n37.RGB, Alpha=@n38), B=ScalarParameter() {defaultValue=1, parameterName=PipShadowOpacity})
  )
  n40 = VectorParameter() {parameterName=StarRanged_Color, defaultValue=<rgba 1 1 1 1>}
  n41 = NamedRerouteUsage()
  n42 = VectorParameter() {parameterName=StarRangedBGColor, defaultValue=<rgba 1 0 0.9092 1>}
  n43 = NamedRerouteUsage()
  n44 = LinearInterpolate(
   A=Multiply(A=LinearInterpolate(B=@n40.RGB, Alpha=@n41), B=ScalarParameter() {defaultValue=1, parameterName=PipRanged_Center_Opacity})
   Alpha=Multiply(A=LinearInterpolate(B=@n42.RGB, Alpha=@n43), B=ScalarParameter() {defaultValue=1, parameterName=PipRangeBG_Opacity})
  )
  NamedRerouteDeclaration(None=LinearInterpolate(
    A=NamedRerouteDeclaration(None=@n39) {name=Pip_Combo_Shadow, nodeColor=<rgba 0.9922 1 0 1>}
    B=NamedRerouteDeclaration(None=@n44) {name=Pip_Combo_Ranged, nodeColor=<rgba 0.0078 0 1 1>}
    Alpha=ScalarParameter() {defaultValue=1, parameterName=Pip_Ranged_ON}
   )) {name=Pip_Master, nodeColor=<rgba 1 0 0.2578 1>}
  NamedRerouteDeclaration(None=MaterialFunctionCall(DistanceField=@n24.SDF, Threshold=Constant() {r=0.5}, Smoothing=Constant() {r=0.02}) {materialFunction=&MF_UI_AntiAliasing}.Outer) {name=Bracket_OpacityMask, nodeColor=<rgba 1 0 0.5391 1>}
  n45 = ScalarParameter() {defaultValue=0.12, parameterName=PipScale_RangedCetner}
  n46 = TextureSample(UVs=MaterialFunctionCall(UVs=@n11.Normalized UV, Texture Scale=LinearInterpolate(A=@n45, B=@n45) {constA=0.125, constB=0.125}) {materialFunction=&ScaleUVsByCenter}.UVs) {samplerType=SAMPLERTYPE_LinearColor, texture=&T_UI_LemonExam_dot}
  NamedRerouteDeclaration(None=MaterialFunctionCall(
    DistanceField=MaterialFunctionCall(MSDF RGB=@n46.RGB) {materialFunction=&MF_UI_MSDFtoSDF}.SDF
    Threshold=Constant() {r=0.58}
    Smoothing=Constant() {r=0.03}
   ) {materialFunction=&MF_UI_AntiAliasing}.Outer) {name=PipRanged_Center, nodeColor=<rgba 0.7812 1 0 1>}
  NamedRerouteDeclaration(None=MaterialFunctionCall(DistanceField=@n17.SDF, Threshold=Constant() {r=0.6}, Smoothing=Constant() {r=0.1}) {materialFunction=&MF_UI_AntiAliasing}.Outer) {name=PipStar_Opacity_Plate, nodeColor=<rgba 0 1 0.25 1>}
  n47 = MaterialFunctionCall(UVs=@n11.Normalized UV, Texture Scale=LinearInterpolate() {constA=0.06, constB=0.06}) {materialFunction=&ScaleUVsByCenter}
  n48 = Constant() {r=0.125}
  NamedRerouteDeclaration(None=MaterialFunctionCall(UVs=MaterialFunctionCall(UVs=@n47.UVs, "Rotation Angle (0-1)"=@n48) {materialFunction=&CustomRotator}.Rotated Values) {materialFunction=&MF_UI_Box}.Fill) {name=PipDefault_OpacityMask, nodeColor=<rgba 0.0547 0 1 1>}
  OneMinus(None=@n14.SDF)
 ]
}
