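# M_UI_LemonExam_TacSprint_Icon is a translucent UI-domain material, named as the tactical sprint icon for LemonExam.
# It samples the sword texture T_UI_LemonExam_Sword with standard texture coordinates as linear color.
# A Sprinter_Color vector parameter in the Color group supplies a tint, light purple by default.
# The graph lists no links from either node to the material outputs.
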Material /LemonExam/UI/Materials/M_UI_LemonExam_TacSprint_Icon
graph {
 properties {blendMode=BLEND_Translucent, materialDomain=MD_UI, pixelDepthOffsetMode=PDOM_Legacy, usedShadingModels=Default Lit}
 nodes [
  TextureSample(UVs=TextureCoordinate()) {samplerType=SAMPLERTYPE_LinearColor, texture=/LemonExam/UI/Textures/T_UI_LemonExam_Sword}
  VectorParameter() {group=Color, parameterName=Sprinter_Color, defaultValue=<rgba 0.6867 0.4621 1 1>}
 ]
}
