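# Translucent UI material for a rounded progress bar, named for LemonExam tac sprint.
# A two-color gradient gains a reflected highlight and a time-scrolled, aspect-corrected Shine band.
# A hard-edged Progress mask splits fill from animated backing color and sets their alphas, while a rotated mask adds an angled marker at the leading edge.
# SimpleProgress and HasBacking static switches choose the composite fed into MF_RadialProgBar as FillColor.
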Material /LemonExam/UI/Materials/M_UI_TacSprintLemonExam_Rounded
graph {
 properties {blendMode=BLEND_Translucent, materialDomain=MD_UI, pixelDepthOffsetMode=PDOM_Legacy, usedShadingModels=Default Lit}
 nodes [
  n1 = LinearInterpolate(
   A=VectorParameter() {group=Color, parameterName=Backing_Color, defaultValue=<rgba 0 0 0 1>}.RGB
   B=VectorParameter() {group=Color, parameterName=Backing_Color_Stop, defaultValue=<rgba 0 0 0 1>}.RGB
   Alpha=ScalarParameter() {parameterName=AnimateBackingColor}
  )
  n2 = VectorParameter() {defaultValue=<rgba 0.6445 0.1714 1 1>}
  n3 = VectorParameter() {parameterName=Param_1, defaultValue=<rgba 0.6867 0.4621 1 1>}
  n4 = MaterialFunctionCall() {materialFunction=&GetUserInterfaceUV}
  n5 = MaterialFunctionCall(UV Channel=@n4.Normalized UV) {materialFunction=&LinearGradient}
  n6 = Constant() {r=0.25}
  n7 = Power(Base=OneMinus(None=MaterialFunctionCall!(UVs, Offset, RotationCenter, RotationAngle=@n6) {materialFunction=/FortUILibrary/MaterialFunctions/Gradients/MF_UI_ReflectedGradient}.Result)) {constExponent=5}
  n8 = Add(A=Add(A=LinearInterpolate(A=@n2.RGB, B=@n3.RGB, Alpha=@n5.UGradient), B=Multiply(A=@n7)), B=NamedRerouteUsage())
  n9 = Add(A=Subtract(A=OneMinus(None=@n4.Normalized UV)), B=ScalarParameter() {defaultValue=1, group=Bar, parameterName=Progress})
  n10 = Saturate(None=Multiply(A=ComponentMask(None=@n9) {r=true}, B=ScalarParameter() {defaultValue=1024, group=Bar, parameterName=Progress_Edge_Hardness}))
  n11 = LinearInterpolate(A=@n1, B=@n8, Alpha=@n10)
  n12 = VectorParameter() {group=Color, parameterName=Marker_Color, defaultValue=<rgba 0.6445 0.1714 1 1>}
  n13 = ScalarParameter() {parameterName=LiftBlendFadeStart}
  n14 = ComponentMask(None=Add(A=TextureCoordinate(), B=ScalarParameter() {parameterName=OffsetBlendFadeStart})) {r=true}
  n15 = Saturate(None=DotProduct(A=LinearInterpolate(A=@n13, Alpha=@n14), B=ScalarParameter() {group=Marker, parameterName=BlendMarkerColor}))
  n16 = ScalarParameter() {defaultValue=-0.0125, group=Marker, parameterName=MarkerAngle_Rotation}
  n17 = Subtract(
   A=OneMinus(None=MaterialFunctionCall(UVs=@n4.Normalized UV, "Rotation Angle (0-1)"=@n16) {materialFunction=&CustomRotator}.Rotated Values)
   B=Add(A=Constant() {r=1}, B=ScalarParameter() {defaultValue=0.0625, group=Marker, parameterName=MarkerAngle_Offset})
  )
  n18 = ScalarParameter() {defaultValue=1, group=Bar, parameterName=Progress}
  n19 = Multiply(A=ComponentMask(None=Add(A=@n17, B=@n18)) {r=true}, B=ScalarParameter() {defaultValue=80, group=Marker, parameterName=Angle_Edge_Hardness})
  n20 = LinearInterpolate(
   A=@n8
   B=LinearInterpolate(A=LinearInterpolate(A=@n8, B=@n12.RGB, Alpha=@n15), B=@n12.RGB, Alpha=ScalarParameter() {parameterName=AnimateMarkerIntensity})
   Alpha=OneMinus(None=Saturate(None=@n19))
  )
  n21 = StaticSwitchParameter(True=LinearInterpolate(A=@n1, B=@n20, Alpha=@n10), False=@n20) {parameterName=HasBacking}
  n22 = ScalarParameter() {defaultValue=0.5, group=Bar, parameterName=Progress_Backing_Alpha}
  n23 = ScalarParameter() {defaultValue=1, group=Bar, parameterName=Progress_Fill_Alpha}
  MaterialFunctionCall!(
   FillColor=Multiply(A=StaticSwitchParameter(True=@n11, False=@n21) {parameterName=SimpleProgress}, B=LinearInterpolate(A=@n22, B=@n23, Alpha=@n10))
   BGColor
  ) {materialFunction=/Game/Athena/HUD/Bacchus/Material/MF_RadialProgBar}
  n24 = Constant() {r=1}
  n25 = TextureCoordinate()
  n26 = Constant2Vector() {g=0.5, r=0.5}
  n27 = MaterialFunctionCall() {materialFunction=&GetUserInterfaceUV}
  n28 = ComponentMask(None=@n27.Pixel Size) {g=true}
  n29 = ComponentMask(None=@n27.Pixel Size) {r=true}
  n30 = Saturate(None=AppendVector(A=Divide(A=@n28, B=@n29), B=Divide(A=@n29, B=@n28)))
  n31 = Multiply(A=Subtract(A=Multiply(A=@n24, B=@n25) {constA=1}, B=@n26), B=Divide(B=Max(A=@n30) {constB=0.001}))
  n32 = Constant() {r=-9}
  n33 = Constant() {r=9}
  n34 = Frac(None=Multiply(A=Constant() {r=0.5}, B=Time()))
  NamedRerouteDeclaration(None=MaterialFunctionCall(UVs=Add(A=@n31, B=@n26), Offset=LinearInterpolate(A=@n32, B=@n33, Alpha=@n34), RotationAngle=Constant() {r=-0.22}) {materialFunction=&MF_UI_ReflectedGradient}.Result) {name=Shine, nodeColor=<rgba 0 0 0 1>}
 ]
}
